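# Sound cue that plays line 012 of the Foundation's "JonesKnows" Tower questline dialogue from the Battlepass S16 voice-over.
# A single Dialogue Player node plays the dialogue wave with the FoundationVO voice as speaker and no targets.
# The cue uses the NPC voice-over sound class, reduced volume and the borrowed Crystal Shard mission attenuation for distance falloff.
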SoundCue /BattlepassS16/Sounds/VO/Foundation/FOUNDATION_2021_TowerQuestline_JonesKnows_012_Dialogue_Cue
root=DialoguePlayer_0
nodes [{
  id=DialoguePlayer_0
  properties {dialogueWaveParameter {
    dialogueWave=/BattlepassS16/Sounds/VO/Foundation/Waves/FOUNDATION_2021_TowerQuestline_JonesKnows_012_Dialogue
    context {speaker=/Game/Sounds/Fort_Human_VO_BR/DialogueVoices/FoundationVO, targets []}
   }}
 }]
properties {
 attenuationSettings=/Game/Sounds/Attenuation/Fort_Mission_CrystalShard_Att
 maxDistance=2800
 soundClassObject=&Fort_VO_NPC
 volumeMultiplier=0.6667
}
